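# Physics asset for the left-side ZirconSweep Axe pickaxe mesh, a cosmetic melee weapon.
# A single capsule body on the Melee_TwoHanded_Handle bone, offset and rotated from that bone, forms the whole collision shape.
# It has no constraints, so no joints link multiple bodies.
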PhysicsAsset /BRCosmetics/Weapons/FORT_Melee/Pickaxe_ZirconSweep_Axe/Meshes/ZirconSweep_Axe_L_PhysicsAsset
bodies {Melee_TwoHanded_Handle [{shapeType=Capsule, length=39.0973, radius=12.5027, center=<-3.9678 -8.5161 41.1677>, rotation=<rot -0.734 -90 -4.6191>}]}
constraints []
