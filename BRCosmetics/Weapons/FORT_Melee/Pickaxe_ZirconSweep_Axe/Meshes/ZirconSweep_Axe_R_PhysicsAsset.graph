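# Physics asset for the right-hand ZirconSweep Axe pickaxe, a melee weapon cosmetic in Battle Royale.
# A single capsule body on the Melee_TwoHanded_Handle bone provides the collision shape, offset and rotated along that bone.
# It has no constraints, so the asset acts as one rigid piece with no articulated parts.
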PhysicsAsset /BRCosmetics/Weapons/FORT_Melee/Pickaxe_ZirconSweep_Axe/Meshes/ZirconSweep_Axe_R_PhysicsAsset
bodies {Melee_TwoHanded_Handle [{shapeType=Capsule, length=39.0717, radius=12.5041, center=<-4.0055 -7.7654 41.1895>, rotation=<rot 0.676 -90 -4.5346>}]}
constraints []
